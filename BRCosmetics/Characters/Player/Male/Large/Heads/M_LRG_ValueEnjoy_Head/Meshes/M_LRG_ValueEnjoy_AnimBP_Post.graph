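# Post-process AnimBlueprint for the large male ValueEnjoy head that adds secondary jiggle to the face.
# The input pose runs through four chained Anim Dynamics nodes for the left and right jawline and nasolabial bulge bones.
# Spring-driven, root-relative simulation uses gravity overrides and tight limits: jawlines translate and rotate slightly, while bulges only translate.
# The event graph holds no logic.
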AnimBlueprint /BRCosmetics/Characters/Player/Male/Large/Heads/M_LRG_ValueEnjoy_Head/Meshes/M_LRG_ValueEnjoy_AnimBP_Post
target_skeleton=/Game/Characters/Player/Common/Fortnite_Base_Head/Fortnite_Base_3L_Head_Skeleton
graphs [
 AnimGraph {nodes [
   n1 = Anim Dynamics - Bone: FACIAL_L_Jawline!(ComponentPose=Local To Component(LocalPose=Input Pose()), bAlphaBoolEnabled:bool=True, GravityOverride:Vector="0,0,-200") -> (Pose) {
    alphaInputType=Bool
    angularBiasOverride=1
    angularDampingOverride=1
    angularSpringConstant=5000
    bAngularSpring=true
    bGravityOverrideInSimSpace=true
    bLinearSpring=true
    bOverrideAngularBias=true
    bOverrideAngularDamping=true
    bOverrideLinearDamping=true
    bUseGravityOverride=true
    linearDampingOverride=1
    linearSpringConstant=100
    simulationSpace=RootRelative
    boundBone {boneName=FACIAL_L_Jawline}
    componentAppliedLinearAccClamp=<5 5 5>
    gravityOverride {z=-200}
    physicsBodyDefinitions [{
      boundBone {boneName=FACIAL_L_Jawline}
      boxExtents=<80 80 80>
      constraintSetup {angularLimitsMax=<0 5 5>, angularLimitsMin=<0 -5 -5>, angularTarget=<1 0 0>, linearAxesMax=<0 0.6 0.6>, linearAxesMin=<0 -0.6 -0.6>}
      localJointOffset=<1 0 0>
     }]
    relativeSpaceBone {boneName=spine_04}
    simSpaceSettings {maxAngularAcceleration=50, maxAngularVelocity=10, simSpaceAngularAlpha=0.5}
   }
   n2 = Anim Dynamics - Bone: FACIAL_R_Jawline!(ComponentPose=@n1, bAlphaBoolEnabled:bool=True, GravityOverride:Vector="0,0,-200") -> (Pose) {
    alphaInputType=Bool
    angularBiasOverride=1
    angularDampingOverride=1
    angularSpringConstant=5000
    bAngularSpring=true
    bGravityOverrideInSimSpace=true
    bLinearSpring=true
    bOverrideAngularBias=true
    bOverrideAngularDamping=true
    bOverrideLinearDamping=true
    bUseGravityOverride=true
    linearDampingOverride=1
    linearSpringConstant=100
    simulationSpace=RootRelative
    boundBone {boneName=FACIAL_R_Jawline}
    componentAppliedLinearAccClamp=<5 5 5>
    gravityOverride {z=-200}
    physicsBodyDefinitions [{
      boundBone {boneName=FACIAL_R_Jawline}
      boxExtents=<90 90 90>
      constraintSetup {angularLimitsMax=<0 5 5>, angularLimitsMin=<0 -5 -5>, angularTarget=<1 0 0>, linearAxesMax=<0 0.6 0.6>, linearAxesMin=<0 -0.6 -0.6>}
      localJointOffset=<-1 0 0>
     }]
    relativeSpaceBone {boneName=spine_04}
    simSpaceSettings {maxAngularAcceleration=50, maxAngularVelocity=10, simSpaceAngularAlpha=0.5}
   }
   n3 = Anim Dynamics - Bone: FACIAL_L_NasolabialBulge!(ComponentPose=@n2, bAlphaBoolEnabled:bool=True, GravityOverride:Vector="0,0,-200") -> (Pose) {
    alphaInputType=Bool
    angularBiasOverride=1
    angularDampingOverride=1
    angularSpringConstant=5000
    bGravityOverrideInSimSpace=true
    bLinearSpring=true
    bOverrideAngularBias=true
    bOverrideAngularDamping=true
    bOverrideLinearDamping=true
    bUseGravityOverride=true
    linearDampingOverride=1
    linearSpringConstant=100
    simulationSpace=RootRelative
    boundBone {boneName=FACIAL_L_NasolabialBulge}
    componentAppliedLinearAccClamp=<5 5 5>
    gravityOverride {z=-200}
    physicsBodyDefinitions [{
      boundBone {boneName=FACIAL_L_NasolabialBulge}
      boxExtents=<80 80 80>
      constraintSetup {angularLimitsMax=<0 0 0>, angularLimitsMin=<0 0 0>, angularTarget=<1 0 0>, linearAxesMax=<0 0.2 0.2>, linearAxesMin=<0 -0.2 -0.2>}
      localJointOffset=<1 0 0>
     }]
    relativeSpaceBone {boneName=spine_04}
    simSpaceSettings {maxAngularAcceleration=50, maxAngularVelocity=10, simSpaceAngularAlpha=0.5}
   }
   n4 = Anim Dynamics - Bone: FACIAL_R_NasolabialBulge!(ComponentPose=@n3, bAlphaBoolEnabled:bool=True, GravityOverride:Vector="0,0,-200") -> (Pose) {
    alphaInputType=Bool
    angularBiasOverride=1
    angularDampingOverride=1
    angularSpringConstant=5000
    bGravityOverrideInSimSpace=true
    bLinearSpring=true
    bOverrideAngularBias=true
    bOverrideAngularDamping=true
    bOverrideLinearDamping=true
    bUseGravityOverride=true
    linearDampingOverride=1
    linearSpringConstant=100
    simulationSpace=RootRelative
    boundBone {boneName=FACIAL_R_NasolabialBulge}
    componentAppliedLinearAccClamp=<5 5 5>
    gravityOverride {z=-200}
    physicsBodyDefinitions [{
      boundBone {boneName=FACIAL_R_NasolabialBulge}
      boxExtents=<90 90 90>
      constraintSetup {angularLimitsMax=<0 0 0>, angularLimitsMin=<0 0 0>, angularTarget=<1 0 0>, linearAxesMax=<0 0.2 0.2>, linearAxesMin=<0 -0.2 -0.2>}
      localJointOffset=<-1 0 0>
     }]
    relativeSpaceBone {boneName=spine_04}
    simSpaceSettings {maxAngularAcceleration=50, maxAngularVelocity=10, simSpaceAngularAlpha=0.5}
   }
   Output Pose(Result=Component To Local(ComponentPose=@n4))
  ]}
 EventGraph {nodes [
   TryGetPawnOwner()
   Event BlueprintUpdateAnimation()
  ]}
]
variables []
